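# Physics asset for the face accessory part of the male medium Agent Ace Alter Ego player body, defining physics for its hair bones.
# Identical small capsules sit on the top, back, front and left-suffixed hair bones, with another capsule on the head.
# Each hair body is constrained directly to the head with narrow swing limits and zero-degree twist limits.
# The exception is front_c_b, which chains off front_c_a, both with a tighter second swing limit.
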PhysicsAsset /BRCosmetics/Characters/Player/Male/Medium/Bodies/M_MED_Agent_Ace_Alter_Ego/Meshe/Parts/M_MED_Agent_Ace_Alter_Ego_FaceAcc_Physics
bodies {
 dyn_hair_a_l [{shapeType=Capsule, length=1.5, radius=1.5, center=<2 0 0>, rotation=<rot 90 0 0>}]
 dyn_hair_b_l [{shapeType=Capsule, length=1.5, radius=1.5, center=<2 0 0>, rotation=<rot 90 0 0>}]
 dyn_hair_back_a [{shapeType=Capsule, length=1.5, radius=1.5, center=<2 0 0>, rotation=<rot 90 0 0>}]
 dyn_hair_back_b [{shapeType=Capsule, length=1.5, radius=1.5, center=<2 0 0>, rotation=<rot 90 0 0>}]
 dyn_hair_back_c [{shapeType=Capsule, length=1.5, radius=1.5, center=<2 0 0>, rotation=<rot 90 0 0>}]
 dyn_hair_back_d [{shapeType=Capsule, length=1.5, radius=1.5, center=<2 0 0>, rotation=<rot 90 0 0>}]
 dyn_hair_back_e [{shapeType=Capsule, length=1.5, radius=1.5, center=<2 0 0>, rotation=<rot 90 0 0>}]
 dyn_hair_back_f [{shapeType=Capsule, length=1.5, radius=1.5, center=<2 0 0>, rotation=<rot 90 0 0>}]
 dyn_hair_front_a [{shapeType=Capsule, length=1.5, radius=1.5, center=<2 0 0>, rotation=<rot 90 0 0>}]
 dyn_hair_front_b [{shapeType=Capsule, length=1.5, radius=1.5, center=<2 0 0>, rotation=<rot 90 0 0>}]
 dyn_hair_front_c_a [{shapeType=Capsule, length=1.5, radius=1.5, center=<2 0 0>, rotation=<rot 90 0 0>}]
 dyn_hair_front_c_b [{shapeType=Capsule, length=1.5, radius=1.5, center=<2 0 0>, rotation=<rot 90 0 0>}]
 dyn_hair_top_a [{shapeType=Capsule, length=1.5, radius=1.5, center=<2 0 0>, rotation=<rot 90 0 0>}]
 dyn_hair_top_b [{shapeType=Capsule, length=1.5, radius=1.5, center=<2 0 0>, rotation=<rot 90 0 0>}]
 dyn_hair_top_c [{shapeType=Capsule, length=1.5, radius=1.5, center=<2 0 0>, rotation=<rot 90 0 0>}]
 dyn_hair_top_d [{shapeType=Capsule, length=1.5, radius=1.5, center=<2 0 0>, rotation=<rot 90 0 0>}]
 dyn_hair_top_e [{shapeType=Capsule, length=1.5, radius=1.5, center=<2 0 0>, rotation=<rot 90 0 0>}]
 dyn_hair_top_f [{shapeType=Capsule, length=1.5, radius=1.5, center=<2 0 0>, rotation=<rot 90 0 0>}]
 head [{shapeType=Capsule, length=1, radius=1}]
}
constraints [
 {bone1Name=dyn_hair_top_a, bone2Name=head, swing1LimitDegrees=15, swing2LimitDegrees=15, twistLimitDegrees=0}
 {bone1Name=dyn_hair_top_b, bone2Name=head, swing1LimitDegrees=15, swing2LimitDegrees=15, twistLimitDegrees=0}
 {bone1Name=dyn_hair_top_d, bone2Name=head, swing1LimitDegrees=15, swing2LimitDegrees=15, twistLimitDegrees=0}
 {bone1Name=dyn_hair_top_c, bone2Name=head, swing1LimitDegrees=15, swing2LimitDegrees=15, twistLimitDegrees=0}
 {bone1Name=dyn_hair_back_a, bone2Name=head, swing1LimitDegrees=15, swing2LimitDegrees=15, twistLimitDegrees=0}
 {bone1Name=dyn_hair_back_e, bone2Name=head, swing1LimitDegrees=15, swing2LimitDegrees=15, twistLimitDegrees=0}
 {bone1Name=dyn_hair_back_f, bone2Name=head, swing1LimitDegrees=15, swing2LimitDegrees=15, twistLimitDegrees=0}
 {bone1Name=dyn_hair_b_l, bone2Name=head, swing1LimitDegrees=15, swing2LimitDegrees=15, twistLimitDegrees=0}
 {bone1Name=dyn_hair_a_l, bone2Name=head, swing1LimitDegrees=15, swing2LimitDegrees=15, twistLimitDegrees=0}
 {bone1Name=dyn_hair_back_c, bone2Name=head, swing1LimitDegrees=15, swing2LimitDegrees=15, twistLimitDegrees=0}
 {bone1Name=dyn_hair_back_d, bone2Name=head, swing1LimitDegrees=15, swing2LimitDegrees=15, twistLimitDegrees=0}
 {bone1Name=dyn_hair_back_b, bone2Name=head, swing1LimitDegrees=15, swing2LimitDegrees=15, twistLimitDegrees=0}
 {bone1Name=dyn_hair_top_e, bone2Name=head, swing1LimitDegrees=15, swing2LimitDegrees=15, twistLimitDegrees=0}
 {bone1Name=dyn_hair_top_f, bone2Name=head, swing1LimitDegrees=15, swing2LimitDegrees=15, twistLimitDegrees=0}
 {bone1Name=dyn_hair_front_b, bone2Name=head, swing1LimitDegrees=15, swing2LimitDegrees=15, twistLimitDegrees=0}
 {bone1Name=dyn_hair_front_a, bone2Name=head, swing1LimitDegrees=15, swing2LimitDegrees=15, twistLimitDegrees=0}
 {bone1Name=dyn_hair_front_c_a, bone2Name=head, swing1LimitDegrees=15, swing2LimitDegrees=5, twistLimitDegrees=0}
 {bone1Name=dyn_hair_front_c_b, bone2Name=dyn_hair_front_c_a, swing1LimitDegrees=15, swing2LimitDegrees=5, twistLimitDegrees=0}
]
